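; Interactable chest of drawers prop for the Asteria Neon City set, with an "Open Chest of Drawers" prompt and a fixed interaction time.
; Interaction is allowed only while the CanInteractPlayer flag is true.
; When that flag replicates as false, its OnRep turns off the availability effects by deactivating Niagara, fading out audio and hiding the spotlight.
; The construction script only calls the parent, and no logic here clears the flag or grants loot.
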(fn BlueprintCanInteract
  (params (InteractingPawn :type Fort%20Pawn%20Object%20Reference) (InteractionBeingAttempted :type EInteractionBeingAttempted%20Enum :default "FirstInteraction") (InteractionType :type TInteractionType%20Enum :default "IT_NoInteraction"))
  (return (Variables|Default|GetCanInteractPlayer)))

(fn BlueprintGetInteractionString
  (params (InteractingPawn :type Fort%20Pawn%20Object%20Reference) (InteractionBeingAttempted :type EInteractionBeingAttempted%20Enum :default "FirstInteraction"))
  (return "NSLOCTEXT(\"[8F49EE3A5C8EB94778486A3D3C826922]\", \"49D826A3458CA4289EF6EE8800BB1A96\", \"Open Chest of Drawers\")"))

(fn BlueprintGetInteractionTime
  (params (InteractingPawn :type Fort%20Pawn%20Object%20Reference) (InteractionBeingAttempted :type EInteractionBeingAttempted%20Enum :default "FirstInteraction"))
  (return 6.0 true))

(fn OnRep_CanInteractPlayer
  (bind _audio (Variables|Default|GetAudio))
  (if (not (Variables|Default|GetCanInteractPlayer))
    (Components|Activation|Deactivate (Variables|Default|GetNiagara))
    (Utilities|IsValid _audio
      (:Is%20Valid
        (Audio|Components|Audio|FadeOut _audio 0.2 0.0 "Linear" :!class /Script/Engine.AudioComponent))
      (:Is%20Not%20Valid
        _))
    (Rendering|SetVisibility (Variables|Default|GetSpotLight) false false)))

(fn ConstructionScript
  (|CallParentFunction))
